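# Widget Blueprint for a generic dynamic info text component in the Athena Bacchus HUD, derived from GenericDynamicInfoUserWidget.
# An Overlay holds the inherited Text_DisplayText, a centered text block anchored to the bottom, with negative padding pushing it below that edge.
# The SetViewModel override calls the parent, then stores the incoming view model in GenericDynamicInfoViewModel; the graph itself never writes the text.
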WidgetBlueprint /Game/Athena/HUD/Bacchus/WidgetComponents/WidgetComponent_GenericDynamicInfo
graphs [EventGraph {nodes [
   n1 = Event SetViewModel!() -> (OutputDelegate, then, DynamicInfoViewModel:Generic Dynamic Info View Model Object Reference) {is=title K2Node_Event}
   > Parent: SetViewModel!(execute, DynamicInfoViewModel:Generic Dynamic Info View Model Object Reference=@n1.DynamicInfoViewModel) -> (then) {is=title K2Node_CallParentFunction}
   > Set GenericDynamicInfoViewModel(GenericDynamicInfoViewModel:Generic Dynamic Info View Model Object Reference=@n1.DynamicInfoViewModel) -> (Output_Get:Generic Dynamic Info View Model Object Reference)
  ]}]
variables []
widgets {
 info {inheritedWidgetCount=1, parentClass=/Script/FortniteUI.GenericDynamicInfoUserWidget, rootWidgetClass=&Overlay, widgetCount=2}
 widgets [
  Overlay_38 {widgetClassPath=&Overlay, widget=Overlay_38}
  Text_DisplayText {
   widgetClassPath=&CommonTextBlock
   bInherited=true
   bIsVariable=true
   parent=Overlay_38
   widget=Text_DisplayText
   properties {
    bIsScrollingEnabled=false
    justification=Center
    style=&TextStyle-Base_C
    text=DisplayText
    font {
     bForceMonospaced=false
     bMaterialIsStencil=false
     fontMaterial=null
     fontObject=&NotoSans
     letterSpacing=0
     monospacedWidth=1
     size=24
     skewAmount=0
     typefaceFontName=NotoSans-Regular
     outlineSettings {
      bApplyOutlineToDropShadows=false
      bMiteredCorners=false
      bSeparateFillAlpha=false
      outlineMaterial=null
      outlineSize=0
      outlineColor=<rgba 0 0 0 1>
     }
    }
    shadowOffset=<0 0>
   }
   slot {class=OverlaySlot, properties {horizontalAlignment=HAlign_Center, verticalAlignment=VAlign_Bottom, padding=<ltrb 0 0 0 -150>}}
  }
 ]
}
