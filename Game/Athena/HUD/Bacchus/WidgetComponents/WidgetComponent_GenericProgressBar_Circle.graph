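# Circular progress bar HUD widget for the Bacchus HUD, built on the native GenericProgressBarUserWidget.
# An overlay stacks primary and delta bar images sharing an interaction progress bar material instance, with the delta image hidden.
# Two animations drive the overlay's opacity: a fade-in and a pulse out and back.
# The graph only overrides SetViewModel to call the parent and store the view model; it contains no progress logic.
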WidgetBlueprint /Game/Athena/HUD/Bacchus/WidgetComponents/WidgetComponent_GenericProgressBar_Circle
animations [
 FadeAnim {
  binding_count=1
  length_seconds=1
  object=&self:FadeAnim
  sequence {bindings [ProgressBarsOverlay {
     id=#1
     possessed_class=Overlay
     tracks [{
       type=MovieSceneFloatTrack
       display_name=Render Opacity
       sections [{type=MovieSceneFloatSection, end_frame=20, start_frame=0, channels [None {type=MovieSceneScriptingFloatChannel, keys [[0,0],[20,1]]}]}]
      }]
    }]}
 }
 PulsingAnim {
  binding_count=1
  length_seconds=0.5
  object=&self:PulsingAnim
  sequence {bindings [ProgressBarsOverlay {
     id=#2
     possessed_class=Overlay
     tracks [{
       type=MovieSceneFloatTrack
       display_name=Render Opacity
       sections [{type=MovieSceneFloatSection, end_frame=10, start_frame=0, channels [None {type=MovieSceneScriptingFloatChannel, keys [[0,1],[5,0],[10,1]]}]}]
      }]
    }]}
 }
]
graphs [EventGraph {nodes [
   n1 = Event SetViewModel!() -> (OutputDelegate, then, ViewModel:Generic Progress Bar View Model Object Reference) {is=title K2Node_Event}
   > Parent: SetViewModel!(execute, ViewModel:Generic Progress Bar View Model Object Reference=@n1.ViewModel) -> (then) {is=title K2Node_CallParentFunction}
   > Set GenericProgressBarViewModel(GenericProgressBarViewModel:Generic Progress Bar View Model Object Reference=@n1.ViewModel) -> (Output_Get:Generic Progress Bar View Model Object Reference)
  ]}]
variables []
widgets {
 info {inheritedWidgetCount=3, parentClass=/Script/FortniteUI.GenericProgressBarUserWidget, rootWidgetClass=&Overlay, widgetCount=3}
 widgets [
  ProgressBarsOverlay {widgetClassPath=&Overlay, bInherited=true, bIsVariable=true, widget=ProgressBarsOverlay}
  PrimaryProgressBar {
   widgetClassPath=&Image
   bInherited=true
   bIsVariable=true
   parent=ProgressBarsOverlay
   widget=PrimaryProgressBar
   properties {brush {
     bIsDynamicallyLoaded=false
     drawAs=Image
     imageType=FullColor
     mirroring=NoMirror
     resourceName=None
     resourceObject=/Game/Athena/HUD/Bacchus/Material/MI_BacchusInteractionProgressBar
     tiling=NoTile
     imageSize=<256 256>
     margin=<ltrb 0 0 0 0>
     outlineSettings {
      bUseBrushTransparency=false
      roundingType=HalfHeightRadius
      width=0
      color {colorUseRule=UseColor_Specified, specifiedColor=<rgba 0 0 0 0>}
      cornerRadii=<0 0 0 0>
     }
     tintColor {colorUseRule=UseColor_Specified, specifiedColor=<rgba 1 1 1 1>}
     uVRegion {bIsValid=false, max=<0 0>, min=<0 0>}
    }}
   slot {class=OverlaySlot}
  }
  DeltaProgressBar {
   widgetClassPath=&Image
   bInherited=true
   bIsVariable=true
   parent=ProgressBarsOverlay
   widget=DeltaProgressBar
   properties {
    visibility=Hidden
    brush {
     bIsDynamicallyLoaded=false
     drawAs=Image
     imageType=FullColor
     mirroring=NoMirror
     resourceName=None
     resourceObject=&MI_BacchusInteractionProgressBar
     tiling=NoTile
     imageSize=<256 256>
     margin=<ltrb 0 0 0 0>
     outlineSettings {
      bUseBrushTransparency=false
      roundingType=HalfHeightRadius
      width=0
      color {colorUseRule=UseColor_Specified, specifiedColor=<rgba 0 0 0 0>}
      cornerRadii=<0 0 0 0>
     }
     tintColor {colorUseRule=UseColor_Specified, specifiedColor=<rgba 1 1 1 1>}
     uVRegion {bIsValid=false, max=<0 0>, min=<0 0>}
    }
   }
   slot {class=OverlaySlot}
  }
 ]
}
