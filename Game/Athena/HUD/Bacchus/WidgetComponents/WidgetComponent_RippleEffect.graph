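# WidgetComponent_RippleEffect is a touch ripple effect widget for the Athena Bacchus HUD, built on the native FortRippleEffectUserWidget class.
# It holds an inherited FortTouchRippleImage inside a hit-test-invisible Overlay, drawing the MI_RippleMedium material.
# Negative slot padding lets the ripple extend beyond the widget's bounds.
# The event graph is empty, so the Blueprint only supplies layout and brush, with no logic of its own.
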WidgetBlueprint /Game/Athena/HUD/Bacchus/WidgetComponents/WidgetComponent_RippleEffect
graphs [EventGraph {links [], nodes []}]
variables []
widgets {
 info {inheritedWidgetCount=1, parentClass=/Script/FortniteUI.FortRippleEffectUserWidget, rootWidgetClass=&Overlay, widgetCount=2}
 widgets [
  Overlay_0 {widgetClassPath=&Overlay, widget=Overlay_0, properties {visibility=HitTestInvisible}}
  Image_RippleEffect {
   widgetClassPath=/Script/FortniteUI.FortTouchRippleImage
   bInherited=true
   bIsVariable=true
   parent=Overlay_0
   widget=Image_RippleEffect
   properties {brush {
     bIsDynamicallyLoaded=false
     drawAs=Image
     imageType=FullColor
     mirroring=NoMirror
     resourceName=None
     resourceObject=/Game/Athena/HUD/Bacchus/Material/RippleEffect/MI_RippleMedium
     tiling=NoTile
     imageSize=<256 256>
     margin=<ltrb 0 0 0 0>
     outlineSettings {
      bUseBrushTransparency=false
      roundingType=HalfHeightRadius
      width=0
      color {colorUseRule=UseColor_Specified, specifiedColor=<rgba 0 0 0 0>}
      cornerRadii=<0 0 0 0>
     }
     tintColor {colorUseRule=UseColor_Specified, specifiedColor=<rgba 1 1 1 1>}
     uVRegion {bIsValid=false, max=<0 0>, min=<0 0>}
    }}
   slot {class=OverlaySlot, properties {padding=<ltrb -32 -32 -32 -32>}}
  }
 ]
}
